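# Sound cue for the medieval hammer's swing, playing a whoosh in the game's melee weapon sound class.
# A random node picks one of five equally weighted whoosh waves, and a modulator at the root varies its pitch slightly while leaving volume fixed.
# Playback uses small melee weapon attenuation and the player melee weapon concurrency set.
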SoundCue /Game/Sounds/Fort_Weapons/Medieval/Hammer/MedievalHammer_Swing_Cue
root=Modulator_7
nodes [
 {id=Modulator_7, children [Random_1], properties {pitchMax=1.1, pitchMin=0.9, volumeMax=1, volumeMin=1}}
 {id=Random_1, children [WavePlayer_1,WavePlayer_2,WavePlayer_3,WavePlayer_4,WavePlayer_5], properties {weights [1,1,1,1,1]}}
 {id=WavePlayer_1, properties {soundWaveAssetPtr=&/MedievalHammer_Whoosh_01}}
 {id=WavePlayer_2, properties {soundWaveAssetPtr=&/MedievalHammer_Whoosh_02}}
 {id=WavePlayer_3, properties {soundWaveAssetPtr=&/MedievalHammer_Whoosh_03}}
 {id=WavePlayer_4, properties {soundWaveAssetPtr=&/MedievalHammer_Whoosh_04}}
 {id=WavePlayer_5, properties {soundWaveAssetPtr=&/MedievalHammer_Whoosh_05}}
]
properties {
 attenuationSettings=&Fort_Weapon_Melee_Sml_Att
 maxDistance=2800
 soundClassObject=&Fort_WeaponMelee_PSM
 volumeMultiplier=0.8
 attenuationOverrides {distanceAlgorithm=NaturalSound, falloffDistance=2000, attenuationShapeExtents {x=500}}
 concurrencySet [&Player_Melee_Weapon_Conc]
}
